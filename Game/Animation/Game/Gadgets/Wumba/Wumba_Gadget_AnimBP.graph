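# Animation Blueprint for the Wumba gadget that plays its deploy, idle and weapon upgrade animations.
# The Wumba state machine enters at Deploy and automatically moves to Idle, whose play rate follows PegBoard_Speed; its Closed state has no transitions.
# A second single-state machine supplies the non-looping Upgrade_Weapon pose, applied additively when bHasFinishedUpgrading, set by the UpdateUpgradingVisuals event, is true.
# The blended pose passes through DefaultSlot to the output.
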AnimBlueprint /Game/Animation/Game/Gadgets/Wumba/Wumba_Gadget_AnimBP
target_skeleton=/Game/Gadgets/Assets/Wumba/Meshes/Wumba_Gadget_Item_Wumba_Quad_LOD0_Skeleton
graphs [
 AnimGraph {nodes [
   AnimGraphNode_StateMachine_0 = Wumba() {is=title AnimGraphNode_StateMachine}
   AnimGraphNode_StateMachine_1 = Upgrage() {is=title AnimGraphNode_StateMachine}
   n1 = Get bHasFinishedUpgrading() -> (bHasFinishedUpgrading:bool)
   Output Pose(Result=Slot 'DefaultSlot'(Source=Apply Additive!(Base=@AnimGraphNode_StateMachine_0, Additive=@AnimGraphNode_StateMachine_1, bAlphaBoolEnabled:bool=@n1) -> (Pose) {alphaInputType=Bool, alphaBoolBlend {blendOption=Linear}})) {layerGroup=None}
  ]}
 Wumba {
  entry=Deploy
  path=AnimGraph.AnimGraphNode_StateMachine_0.Wumba
  ids=id
  nodes [
   AnimStateNode_0 = Closed() {is=title AnimStateNode}
   AnimStateNode_1 = Deploy() {is=title AnimStateNode}
   AnimStateNode_2 = Idle()
   AnimStateTransitionNode_1 = AnimStateTransitionNode() {bAutomaticRuleBasedOnSequencePlayerInState=true, crossfadeDuration=0}
  ]
  links [{dst=Idle, rule=AnimGraph.AnimGraphNode_StateMachine_0.Wumba.AnimStateTransitionNode_1.Transition, src=Deploy, via=AnimStateTransitionNode_1}]
 }
 Closed {
  path=AnimGraph.AnimGraphNode_StateMachine_0.Wumba.AnimStateNode_0.Closed
  nodes [Output Animation Pose(Result=Sequence Player 'Closed'() {sequence=&/Closed})]
 }
 Deploy {
  path=AnimGraph.AnimGraphNode_StateMachine_0.Wumba.AnimStateNode_1.Deploy
  nodes [Output Animation Pose(Result=Sequence Player 'Deploy'() {bLoopAnimation=false, sequence=&/Deploy})]
 }
 Idle {
  path=AnimGraph.AnimGraphNode_StateMachine_0.Wumba.AnimStateNode_2.Idle
  nodes [Output Animation Pose(Result=Sequence Player 'Idle'!(PlayRate:float=Get PegBoard_Speed() -> (PegBoard_Speed:double)) -> (Pose) {sequence=&/Idle})]
 }
 Transition {path=AnimGraph.AnimGraphNode_StateMachine_0.Wumba.AnimStateTransitionNode_1.Transition, nodes [Result()]}
 Upgrage {entry=Upgrade, path=AnimGraph.AnimGraphNode_StateMachine_1.Upgrage, ids=id, nodes [AnimStateNode_0 = Upgrade() {is=title AnimStateNode}]}
 Upgrade {
  path=AnimGraph.AnimGraphNode_StateMachine_1.Upgrage.AnimStateNode_0.Upgrade
  nodes [Output Animation Pose(Result="Sequence Player (additive) 'Upgrade_Weapon'"() {bLoopAnimation=false, sequence=&/Upgrade_Weapon})]
 }
 EventGraph {nodes [
   n1 = UpdateUpgradingVisuals!() -> (OutputDelegate, then, NewValue:bool=~) {is=title K2Node_CustomEvent}
   > Set bHasFinishedUpgrading(bHasFinishedUpgrading:bool=@n1.NewValue) -> (Output_Get:bool)
  ]}
]
variables [
 bHasFinishedUpgrading {type=boolean}
 PegBoard_Speed {type=number, default=0.25}
]
